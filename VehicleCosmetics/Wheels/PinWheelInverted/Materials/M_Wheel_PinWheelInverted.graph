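# Masked, lit material for the PinWheelInverted vehicle wheel cosmetic on skeletal meshes, drawing an animated rim and tire through emissive color.
# Packed masks shape panning spiral, noise and Fresnel layers, plus a sine-pulsed glow, all tintable by parameters.
# Vertex color red blends rim into tire and confines an optional LOD opacity mask, enabled by static switch.
# The result passes through the shared vehicle damage and de-rez transition functions.
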Material /VehicleCosmetics/Wheels/PinWheelInverted/Materials/M_Wheel_PinWheelInverted
graph {
 properties {
  bAutomaticallySetUsageInEditor=false
  bUseMaterialAttributes=true
  bUsedWithSkeletalMesh=true
  blendMode=BLEND_Masked
  previewMesh=/VehicleCosmetics/Wheels/PinWheelInverted/Meshes/SK_Wheel_PinWheelInverted
  usedShadingModels=Default Lit
 }
 nodes [
  n1 = TextureSampleParameter2D() {
   parameterName=Rim_M
   samplerType=SAMPLERTYPE_LinearColor
   texture=/VehicleCosmetics/Wheels/PinWheelInverted/Textures/T_PinWheelInverted_Rim_M2
  }
  n2 = VectorParameter() {parameterName=CenterColor, defaultValue=<rgba 0.6172 0 0 1>}
  n3 = Max(
   A=Multiply(A=@n1.R, B=@n2.RGB)
   B=Multiply(A=VectorParameter() {parameterName=PinWheelColor, defaultValue=<rgba 1 1 1 1>}.RGB, B=NamedRerouteUsage())
  )
  n4 = NamedRerouteUsage()
  n5 = NamedRerouteUsage()
  n6 = Max(A=Multiply(A=Max(A=@n4, B=@n5), B=VectorParameter() {parameterName=TirelColor, defaultValue=<rgba 1 1 1 1>}.RGB), B=NamedRerouteUsage())
  n7 = VertexColor()
  n8 = Constant() {r=1}
  n9 = TextureSample() {samplerType=SAMPLERTYPE_LinearColor, texture=/VehicleCosmetics/Wheels/PinWheelInverted/Textures/T_PinWheelInverted_Rim_M_LOD}
  n10 = OneMinus(None=@n7.R)
  n11 = SetMaterialAttributes!(
   MaterialAttributes
   Base Color
   Emissive Color=Multiply(A=LinearInterpolate(A=@n3, B=@n6, Alpha=@n7.R), B=ScalarParameter() {defaultValue=1, parameterName=GlobalBrightness})
   Opacity Mask=StaticSwitchParameter(True=LinearInterpolate(A=@n8, B=@n9.R, Alpha=@n10), False=Constant() {r=1}) {parameterName=LODs_Mask}
   Metallic=ScalarParameter() {defaultValue=1, parameterName=Metallic}
   Roughness=ScalarParameter() {defaultValue=1, parameterName=Roughness}
  ) {attributeSetTypes [
    #1
    #2
    #3
    #4
    #5
   ]}
  MaterialFunctionCall(In=MaterialFunctionCall(Material=@n11) {materialFunction=&MF_BaseVehicle_Damage}.Result) {materialFunction=&MF_VehicleDeRez_Transition}
  n12 = Panner(Coordinate=TextureCoordinate() {uTiling=80}) {speedX=3}
  n13 = Multiply(
   A=ConstantBiasScale(None=TextureSample(UVs=@n12) {texture=/VehicleCosmetics/Wheels/PinWheelInverted/Textures/T_PinWheelInverted_Rim_M}.G) {bias=0.2, scale=2}
   B=@n1.B
  )
  n14 = Add(A=Time(), B=Constant() {r=0.25})
  n15 = NamedRerouteDeclaration(None=ScalarParameter() {defaultValue=1, parameterName=SpiralSpeed}) {name=SpiralSpeed, nodeColor=<rgba 1 0 0.5859 1>}
  n16 = TextureSample(UVs=Panner(Coordinate=TextureCoordinate() {uTiling=2, vTiling=2}, Time=Multiply(A=@n14, B=@n15)) {speedX=0.25}) {texture=&T_PinWheelInverted_Rim_M}
  n17 = ConstantBiasScale(None=@n16.G) {bias=-0.1, scale=1}
  n18 = TextureCoordinate() {uTiling=8, vTiling=2}
  n19 = Multiply(
   A=Multiply(A=Multiply(A=@n13, B=@n17), B=Constant() {r=0.6})
   B=ConstantBiasScale(None=TextureSample(UVs=@n18) {texture=&T_PinWheelInverted_Rim_M}.RGB) {bias=0.05, scale=1}
  )
  n20 = Panner(Coordinate=TextureCoordinate() {uTiling=0.5, vTiling=0.5}) {speedX=0.05}
  n21 = Multiply(
   A=ConstantBiasScale(None=@n16.RGB) {bias=0, scale=2}
   B=ConstantBiasScale(None=TextureSample(UVs=@n20) {texture=/VehicleCosmetics/Wheels/_SharedTextures/FX/Universal_Noise_66_RGB}.B) {bias=0.5, scale=1}
  )
  n22 = TextureCoordinate()
  n23 = Clamp(None=ConstantBiasScale(None=TextureSample(UVs=@n22) {texture=/VehicleCosmetics/Wheels/_SharedTextures/FX/GradientBottom2_T}.G) {bias=-0.015, scale=100})
  n24 = OneMinus(None=@n23)
  n25 = Multiply(A=Multiply(A=Multiply(A=@n21, B=@n24), B=@n1.G), B=Constant() {r=0.4})
  n26 = MaterialFunctionCall!(Normal, Exponent=ScalarParameter() {defaultValue=2, parameterName=fresmel_exp}) {materialFunction=/VehicleCosmetics/Wheels/Cola_Holographic/Materials/MF_Fresnel_Classic}
  n27 = Multiply(A=Multiply(A=Multiply(A=@n23, B=@n26.Result), B=@n1.G), B=Constant() {r=1})
  n28 = AppendVector(A=@n27, B=@n27)
  NamedRerouteDeclaration(None=Max(A=Max(A=@n19, B=@n25), B=AppendVector(A=@n28, B=@n27))) {name=Rim, nodeColor=<rgba 0.0547 0 1 1>}
  n29 = Time()
  n30 = ScalarParameter() {defaultValue=1, parameterName=Timer}
  n31 = Add(A=Sine(None=Multiply(A=@n29, B=@n30)) {period=3}, B=Constant() {r=1})
  n32 = Constant() {r=0.25}
  n33 = TextureSampleParameter2D() {
   parameterName=Tire_M
   samplerType=SAMPLERTYPE_LinearColor
   texture=/VehicleCosmetics/Wheels/PinWheelInverted/Textures/T_PinWheelInverted_Tire_M
  }
  n34 = Multiply(A=ConstantBiasScale(None=Multiply(A=@n31, B=@n32)) {bias=0.1, scale=2}, B=@n33.R)
  n35 = ScalarParameter() {defaultValue=1, parameterName=Glow_Multiplier}
  NamedRerouteDeclaration(None=Multiply(A=Multiply(A=@n34, B=@n35), B=@n2.RGB)) {name=InternalGlow, nodeColor=<rgba 0.2656 1 0 1>}
  NamedRerouteDeclaration(None=@n33.G) {name=TireM_Green_CH, nodeColor=<rgba 0.0696 1 0 1>}
  NamedRerouteDeclaration(None=@n33.B) {name=TireM_Blue_CH, nodeColor=<rgba 0 0.0825 1 1>}
  n36 = TextureCoordinate()
  n37 = Multiply(A=NamedRerouteUsage(), B=Time())
  n38 = TextureCoordinate()
  n39 = TextureSample(UVs=AppendVector(A=ComponentMask(None=@n38) {g=true}, B=ComponentMask(None=@n38) {r=true})) {texture=&Universal_Noise_66_RGB}
  n40 = Multiply(
   A=TextureSample(UVs=Panner(Coordinate=@n36, Time=@n37) {speedX=0.25}) {texture=&T_PinWheelInverted_Rim_M}.G
   B=ConstantBiasScale(None=OneMinus(None=@n39.R)) {bias=-0.85, scale=4}
  )
  n41 = Constant() {r=0.65}
  n42 = Time()
  n43 = Constant() {r=0.45}
  n44 = Panner(Coordinate=@n38, Time=Multiply(A=Add(A=@n42, B=@n43), B=NamedRerouteUsage())) {speedX=0.25}
  n45 = Multiply(
   A=ConstantBiasScale(None=@n39.R) {bias=0, scale=1}
   B=ConstantBiasScale(None=TextureSample(UVs=@n44) {texture=&T_PinWheelInverted_Rim_M}.G) {bias=-0.15, scale=1.5}
  )
  n46 = Constant() {r=0.4}
  n47 = ConstantBiasScale(None=Max(A=Multiply(A=@n40, B=@n41), B=Multiply(A=@n45, B=@n46))) {bias=-0.05, scale=1.2}
  n48 = NamedRerouteUsage()
  NamedRerouteDeclaration(None=Multiply(A=Multiply(A=@n47, B=@n48), B=ScalarParameter() {defaultValue=1, parameterName=TireLines})) {name=TirePart1, nodeColor=<rgba 0.4531 0 1 1>}
  n49 = TextureSample(UVs=Panner(Coordinate=TextureCoordinate() {uTiling=200}) {speedX=-2}) {texture=/VehicleCosmetics/Wheels/_SharedTextures/FX/SharpGradient_01_Heightmap}
  n50 = NamedRerouteUsage()
  n51 = Multiply(
   A=Clamp(None=ConstantBiasScale(None=@n49.G) {bias=-0.1, scale=0.2}) {maxDefault=0.6}
   B=OneMinus(None=MaterialFunctionCall(Normal=@n50) {materialFunction=&MF_Fresnel_Classic}.Result)
  )
  n52 = Panner(Coordinate=TextureCoordinate() {uTiling=0.8, vTiling=0.2}) {speedX=0.2, speedY=0.05}
  n53 = Multiply(
   A=Multiply(A=TextureSample(UVs=@n52) {texture=/VehicleCosmetics/Wheels/_SharedTextures/FX/Distort_RGB_T}.G, B=Constant() {r=4})
   B=Constant() {r=0.2}
  )
  n54 = Multiply(
   A=Multiply(A=ScalarParameter() {defaultValue=0.15, parameterName=Tire_Brightness}, B=@n50)
   B=Multiply(A=@n50, B=ConstantBiasScale(None=@n53) {bias=0.15, scale=1})
  )
  NamedRerouteDeclaration(None=Subtract(A=Add(A=@n51, B=@n54), B=NamedRerouteUsage())) {name=TirePart2, nodeColor=<rgba 0 1 0.1328 1>}
 ]
}
